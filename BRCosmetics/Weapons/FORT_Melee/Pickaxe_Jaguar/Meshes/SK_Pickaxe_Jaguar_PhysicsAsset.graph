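# Physics asset for the Jaguar pickaxe skeletal mesh, a Fortnite two-handed melee weapon cosmetic.
# It holds a single capsule body on the Melee_TwoHanded_Handle bone, offset and rotated from that bone.
# It defines no constraints, so the mesh has one rigid collision shape and no articulated parts.
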PhysicsAsset /BRCosmetics/Weapons/FORT_Melee/Pickaxe_Jaguar/Meshes/SK_Pickaxe_Jaguar_PhysicsAsset
bodies {Melee_TwoHanded_Handle [{shapeType=Capsule, length=68.0037, radius=31.8552, center=<0.573 5.779 25.1612>, rotation=<rot -5.8746 -90 0.1137>}]}
constraints []
